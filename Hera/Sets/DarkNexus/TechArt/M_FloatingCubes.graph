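# Unlit additive material for floating cubes in the DarkNexus set, with a glowing, glitchy monitor-static look.
# Emissive blends two colors and combines layers built in the graph, which include fresnel, borders, sweeps, panning static and scanlines.
# Pivot location plus time gives per-cube variation, world position offset includes a sine motion, and opacity fades near the camera.
# A ray-tracing switch picks eye-adaptation-compensated or Lumen-scaled emissive before the global emissive multiplier.
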Material /Hera/Sets/DarkNexus/TechArt/M_FloatingCubes
graph {
 properties {bAutomaticallySetUsageInEditor=false, bUseMaterialAttributes=true, blendMode=BLEND_Additive, shadingModel=MSM_Unlit, usedShadingModels=Unlit}
 nodes [
  n1 = DotProduct(A=CameraVectorWS(), B=VertexNormalWS())
  n2 = Multiply(A=Power(Base=OneMinus(None=@n1)) {constExponent=3}, B=ScalarParameter() {defaultValue=0.5, parameterName=Frensnel Intensity}) {constB=0.1}
  Add(A=@n2) {constB=0.65}
  n3 = MaterialFunctionCall() {materialFunction=&ObjectPivotPoint}
  n4 = MaterialFunctionCall(Float3=Frac(None=Multiply(A=@n3.Object Pivot Location) {constB=0.15})) {materialFunction=&BreakOutFloat3Components}
  n5 = Add(A=@n4.R, B=@n4.R)
  n6 = NamedRerouteDeclaration(None=Add(A=Add(A=@n5, B=@n4.B), B=Time())) {name=Random TIme, nodeColor=<rgba 0.5 1 0 1>}
  n7 = Desaturation(None=MaterialFunctionCall(Hue Shift Percentage=Multiply(A=@n6) {constB=0.3}, Texture=VertexColor()) {materialFunction=&HueShift}.Result)
  Add(A=@n6, B=@n7)
  n8 = NamedRerouteUsage()
  n9 = EyeAdaptationInverse(LightValueInput=@n8, AlphaInput=Constant() {r=0.5})
  n10 = Multiply(
   A=@n8
   B=Multiply(
    A=VectorParameter() {group=A__EmissiveFilters, parameterName=Emissive_Color_Multiplier, defaultValue=<rgba 1 1 1 1>}.RGB
    B=ScalarParameter() {defaultValue=1, parameterName=LumenEmissiveMultiplier}
   )
  )
  n11 = Divide(A=Subtract(A=PixelDepth().R, B=Constant() {r=200}), B=Constant() {r=500}) {constB=100}
  n12 = Multiply(A=AppendVector(A=NamedRerouteUsage(), B=Constant())) {constB=500}
  n13 = Multiply(A=Sine(None=NamedRerouteUsage()) {period=5}) {constB=20}
  MaterialFunctionCall!(In=SetMaterialAttributes!(
    MaterialAttributes
    Emissive Color=RayTracingQualitySwitch(Normal=@n9.EyeAdaptationInverse, RayTraced=@n10)
    Opacity=Saturate(None=@n11)
    World Position Offset=Add(A=@n12, B=@n13)
   ) {attributeSetTypes [#1, #2, #3]}) {materialFunction=/Game/Global/MaterialFunctions/MF_GlobalEmissiveMultiplier_MPC}
  n14 = ComponentMask(None=NamedRerouteUsage()) {g=true}
  Multiply(A=@n14) {constB=2}
  n15 = LinearInterpolate(
   A=VectorParameter() {parameterName=Color, defaultValue=<rgba 0.8958 0.2804 0 1>}.RGB
   B=VectorParameter() {parameterName=Color 2, defaultValue=<rgba 0.8958 0.6076 0 1>}.RGB
   Alpha=NamedRerouteUsage()
  )
  n16 = ScalarParameter() {defaultValue=2, parameterName=Emissive Intensity}
  n17 = Add(A=NamedRerouteUsage(), B=NamedRerouteUsage())
  n18 = NamedRerouteUsage()
  n19 = Add(A=Add(A=Add(A=@n17, B=@n18), B=NamedRerouteUsage()), B=NamedRerouteUsage())
  NamedRerouteDeclaration(None=Multiply(A=Multiply(A=@n15, B=@n16), B=Add(A=@n19) {constB=0.05})) {name=Emissive, nodeColor=<rgba 0 1 0.1328 1>}
  n20 = Multiply(A=NamedRerouteUsage(), B=ScalarParameter() {defaultValue=-0.1, parameterName=Sweep Time}) {constB=-0.2}
  n21 = Power(Base=Frac(None=Add(A=@n14, B=@n20))) {constExponent=5}
  NamedRerouteDeclaration(None=Multiply(A=@n21, B=ScalarParameter() {defaultValue=0.5, group=MonitorStatic, parameterName=Sweep Add}) {constB=0.2}) {name=Sweep, nodeColor=<rgba 0.2422 0 1 1>}
  n22 = Multiply(A=MaterialFunctionCall!(Position, UVRadialScale) {materialFunction=/Game/Weapons/WeaponSkins/MaterialFunction/ObjectStabilizedScreenSpaceUVs}.StabilizedUVs) {constB=4}
  n23 = NamedRerouteUsage()
  n24 = Constant2Vector() {g=0.1, r=5}
  n25 = NamedRerouteUsage()
  n26 = TextureSample(UVs=Panner(Coordinate=@n22, Time=@n23, Speed=Multiply(A=@n24, B=@n25)) {speedX=5, speedY=0.1}) {
   samplerSource=SSM_Wrap_WorldGroupSettings
   samplerType=SAMPLERTYPE_Masks
   texture=/Game/Missions/Primary/RepairTheShelter/Textures/Cinematic_Screen_Static
  }
  n27 = Multiply(A=@n26.B) {constB=0.1}
  n28 = Constant2Vector() {g=-0.05, r=6}
  n29 = NamedRerouteUsage()
  n30 = TextureSample(UVs=Panner(Coordinate=@n22, Time=@n23, Speed=Multiply(A=@n28, B=@n29)) {speedX=6, speedY=-0.05}) {
   samplerSource=SSM_Wrap_WorldGroupSettings
   samplerType=SAMPLERTYPE_Masks
   texture=&Cinematic_Screen_Static
  }
  n31 = Multiply(A=@n30.G, B=@n26.G)
  NamedRerouteDeclaration(None=Multiply(A=Add(A=@n27, B=@n31), B=ScalarParameter() {defaultValue=5, group=MonitorStatic, parameterName=Static Scale})) {name=Static, nodeColor=<rgba 0 1 0.8828 1>}
  n32 = Multiply(A=@n30.B, B=@n26.R)
  NamedRerouteDeclaration(None=Multiply(A=AppendVector(A=@n31, B=@n32), B=ScalarParameter() {defaultValue=0.5, group=MonitorStatic, parameterName=Static UV Distort})) {name=StaticUVDistrotion, nodeColor=<rgba 1 0.5625 0 1>}
  n33 = Divide(A=WorldPosition().Z) {constB=125}
  n34 = ScalarParameter() {defaultValue=0.25, parameterName=Noise Size}
  NamedRerouteDeclaration(None=Add(A=Multiply(A=@n33, B=@n34), B=NamedRerouteUsage())) {name=TexCoord, nodeColor=<rgba 1 0 0.4453 1>}
  n35 = Add(A=@n26.RGB, B=@n30.RGB)
  NamedRerouteDeclaration(None=Add(A=Multiply(A=@n35, B=@n30.R), B=@n26.G)) {name=Static Color, nodeColor=<rgba 0.3359 1 0 1>}
  n36 = ScalarParameter() {defaultValue=0.01, group=MonitorStatic, parameterName=Sweep UV Distort}
  NamedRerouteDeclaration(None=AppendVector(A=Multiply(A=@n21, B=@n36) {constB=0.02}, B=Constant())) {name=SweepUVDistortion, nodeColor=<rgba 0.0781 1 0 1>}
  NamedRerouteDeclaration(None=@n2) {name=Fresnel, nodeColor=<rgba 1 0.4195 0 1>}
  n37 = TextureCoordinate()
  n38 = Multiply(A=Add(A=Add(A=@n37) {constB=-0.5}, B=NamedRerouteUsage())) {constB=4}
  n39 = Saturate(None=Add(A=Abs(None=@n38), B=ScalarParameter() {defaultValue=-1.8383, parameterName=Border Size}) {constB=-1})
  n40 = Multiply(A=@n39, B=@n39)
  n41 = Constant()
  NamedRerouteDeclaration(None=Multiply(A=Distance(A=@n40, B=@n41), B=ScalarParameter() {defaultValue=70, parameterName=Border Intensity})) {name=Border, nodeColor=<rgba 0 0.25 1 1>}
  NamedRerouteDeclaration(None=Saturate(None=Multiply(A=@n7) {constB=0.2})) {name=PerCubeTime, nodeColor=<rgba 1 0.4688 0 1>}
  NamedRerouteDeclaration(None=ScalarParameter() {defaultValue=1, group=MonitorStatic, parameterName=StaticSpeed}) {name=StaticSpeed, nodeColor=<rgba 0 1 0.6484 1>}
  n42 = WorldPosition()
  n43 = MaterialFunctionCall!(
   ScanLine_Tiling_1=ScalarParameter() {defaultValue=0.89}
   Scanline_PannerSpeed_1
   ScanLine_Tiling=ScalarParameter() {defaultValue=32.3302, parameterName=Param_1}
   Scanline_PannerSpeed
   UVs=Add(A=Divide(A=@n42.Z) {constB=125}, B=NamedRerouteUsage())
  ) {materialFunction=/Game/Weapons/WeaponSkins/MaterialFunction/ScanLines}
  n44 = Constant()
  n45 = MaterialFunctionCall!(
   Blend Range=ScalarParameter() {defaultValue=173.3058, group=MonitorStatic, parameterName=ScanlineDistanceFadeRange}
   Start Offset=Multiply(A=ScalarParameter() {defaultValue=903.6458, group=MonitorStatic, parameterName=ScanlineDistanceFadeStart}) {constB=-1}
  ) {materialFunction=/Engine/Functions/Engine_MaterialFunctions03/Particles/Distance_Blend}
  NamedRerouteDeclaration(None=Multiply(A=LinearInterpolate(A=@n43.Result, B=@n44, Alpha=@n45.Result)) {constB=0.05}) {name=Scanlines_02, nodeColor=<rgba 0.4766 1 0 1>}
  NamedRerouteUsage()
  NamedRerouteUsage()
  Panner(Coordinate=@n22, Time=@n23, Speed=Multiply(A=Constant2Vector() {g=-0.03, r=-3}, B=NamedRerouteUsage())) {speedX=-3, speedY=-0.03}
  ScalarParameter() {defaultValue=1, parameterName=Opacity}
  Sine(None=Add(A=Time(), B=VertexColor().R))
  TextureCoordinate()
  Time()
 ]
}
