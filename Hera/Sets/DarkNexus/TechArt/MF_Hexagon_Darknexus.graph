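# Material function that turns input material attributes into a dark, glossy, animated glowing hexagon surface for the DarkNexus set.
# A tiled, channel-packed hex texture, UV-offset by a hash of instance position, supplies pattern masks alongside animated tile flicker, two local-height sine sweeps, Fresnel and custom-primitive-data gradient masks.
# Emissive blends hex and highlight colors with optional texture or UV edge darkening.
# A ray tracing quality switch separates the eye-adaptation-compensated glow from a Lumen-scaled one.
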MaterialFunction /Hera/Sets/DarkNexus/TechArt/MF_Hexagon_Darknexus
graph {nodes [
  n1 = ScalarParameter() {defaultValue=1.25, group=SL_S30_Hexagon, parameterName=OutlinePower}
  Add(A=@n1, B=Multiply(A=NamedRerouteUsage()) {constB=1.5})
  Add()
  Constant() {r=0.5}
  Constant()
  n2 = SetMaterialAttributes!(
   MaterialAttributes=NamedRerouteUsage()
   BaseColor=Constant()
   EmissiveColor=NamedRerouteUsage()
   Specular=ScalarParameter() {defaultValue=1, parameterName=Specular}
   Roughness=ScalarParameter() {defaultValue=0.05, parameterName=Roughness}
   Metallic=Constant()
   Normal
  ) {attributeSetTypes [
    #1
    #2
    #3
    #4
    #5
    #6
   ]}
  FunctionOutput(None=@n2)
  n3 = ScalarParameter() {defaultValue=50, group=SL_S30_Hexagon, parameterName=FresnelExp}
  n4 = PixelNormalWS()
  Multiply(
   A=Power(
    Base=Fresnel(ExponentIn=@n3, Normal=@n4) {baseReflectFraction=0, exponent=0.5}
    Exp=ScalarParameter() {defaultValue=5, group=SL_S30_Hexagon, parameterName=FresnelPower}
   )
   B=ScalarParameter() {group=SL_S30_Hexagon, parameterName=FresnelMult}
  )
  n5 = StaticBool() {value=true}
  n6 = ScalarParameter() {defaultValue=1, parameterName=FresnelInverted_Exp}
  NamedRerouteDeclaration(None=Power(
    Base=MaterialFunctionCall(Invert Fresnel=@n5, Power=@n6) {materialFunction=&Fresnel_Function}.Result
    Exp=ScalarParameter() {defaultValue=1, parameterName=FresnelInverted_MaskPower}
   )) {name=InvertedFresnel, nodeColor=<rgba 0 1 0.625 1>}
  NamedRerouteDeclaration(None=FunctionInput() {bUsePreviewValueAsDefault=true, inputName=In_1, inputType=FunctionInput_MaterialAttributes}) {name=InMaterialAttributes_1, nodeColor=<rgba 1 0.3047 0 1>}
  n7 = VertexInterpolator(VS=LocalPosition() {localOrigin=InstancePreSkinning}.XYZ)
  n8 = Add(A=NamedRerouteUsage(), B=NamedRerouteUsage())
  n9 = ScalarParameter() {defaultValue=100, parameterName=TopDownHashNoiseAmount}
  n10 = Divide(A=Add(A=ComponentMask(None=@n7.PS) {b=true}, B=Multiply(A=@n8, B=@n9))) {constB=100}
  n11 = Divide(A=@n10, B=ScalarParameter() {defaultValue=10, group=SL_S30_Hexagon, parameterName=HexSweepScale}) {constB=10}
  n12 = Time()
  n13 = ScalarParameter() {defaultValue=0.1, group=SL_S30_Hexagon, parameterName=HexSweepSpeed}
  n14 = Sine(None=Add(A=NamedRerouteDeclaration(None=@n11) {name=Height_1, nodeColor=<rgba 1 0.1875 0 1>}, B=Multiply(A=@n12, B=@n13) {constB=0.1}))
  n15 = Saturate(None=Power(Base=Abs(None=@n14)) {constExponent=50})
  n16 = Add(A=Divide(A=@n10)) {constB=15}
  n17 = Multiply(
   A=Time()
   B=Divide(
    A=ScalarParameter() {defaultValue=0.01, group=SL_S30_Hexagon, parameterName=HexSweepSpeed2}
    B=ScalarParameter() {defaultValue=10, group=SL_S30_Hexagon, parameterName=HexSweepScale2}
   ) {constB=20}
  ) {constB=0.05}
  n18 = Abs(None=Sine(None=Add(A=@n16, B=@n17)))
  n19 = NamedRerouteDeclaration(None=Saturate(None=Power(Base=@n18) {constExponent=50})) {name=SecondarySweep_1, nodeColor=<rgba 0.8984 0 1 1>}
  n20 = NamedRerouteDeclaration(None=Add(A=NamedRerouteDeclaration(None=@n15) {name=MainSweep_1, nodeColor=<rgba 1 0.0234 0 1>}, B=Multiply(A=@n19) {constB=0.5})) {name=FullSweep_1, nodeColor=<rgba 0.8281 0 1 1>}
  n21 = NamedRerouteUsage()
  NamedRerouteDeclaration(None=Multiply(A=Multiply(A=@n20, B=@n21), B=ScalarParameter() {defaultValue=0.25, group=SL_S30_Hexagon, parameterName=TopdownGradientHashIntensity})) {name=TopdownGradientHash_1, nodeColor=<rgba 0 0.5781 1 1>}
  n22 = TextureSampleParameter2D() {parameterName=TextureMapDarkening, texture=&T_BlacksRGB}
  n23 = ComponentMask(None=TextureCoordinate()) {r=true}
  n24 = ScalarParameter() {defaultValue=0.5, parameterName=VerticalUEdgeDarkeningCenter}
  n25 = ScalarParameter() {defaultValue=0.45, parameterName=VerticalUEdgeDarkeningLength}
  n26 = StaticSwitchParameter(
   True=Power(
    Base=ChannelMaskParameter(None=@n22.RGB) {maskChannel=Green, parameterName=TextureMapDarkeningChannel, defaultValue=<rgba 0 1 0 0>}.R
    Exp=ScalarParameter() {defaultValue=3, parameterName=TextureMapDarkeningPower}
   )
   False=Power(
    Base=SphereMask(A=@n23, B=@n24, Radius=@n25) {attenuationRadius=0.5, hardnessPercent=0}
    Exp=ScalarParameter() {defaultValue=0.65, parameterName=VerticalUVEdgeDarkeningPower}
   )
  ) {desc="True=UseTextureMap, False=UseVerticalUVMask", parameterName=UseTextureMapForDarkening}
  n27 = Constant() {r=1}
  n28 = NamedRerouteUsage()
  n29 = NamedRerouteUsage()
  n30 = Multiply(
   A=NamedRerouteUsage()
   B=LinearInterpolate(
    A=VectorParameter() {group=SL_S30_Hexagon, parameterName=HexColor, defaultValue=<rgba 0 0.2026 0.7344 1>}.RGB
    B=VectorParameter() {group=SL_S30_Hexagon, parameterName=HexHighlightColor, defaultValue=<rgba 0 2.3148 14 1>}.RGB
    Alpha=Add(A=@n28, B=@n29)
   )
  )
  n31 = Multiply(A=Multiply(A=StaticSwitchParameter(True=@n26, False=@n27) {parameterName=UseEdgeDarkening}, B=@n30), B=NamedRerouteUsage())
  n32 = ScalarParameter() {defaultValue=0.3, parameterName=EyeAdaptation Inverse, sliderMax=1}
  n33 = Multiply(
   A=VectorParameter() {group=A__EmissiveFilters, parameterName=Emissive_Color_Multiplier, defaultValue=<rgba 1 1 1 1>}.RGB
   B=ScalarParameter() {defaultValue=1, parameterName=LumenEmissiveMultiplier}
  )
  NamedRerouteDeclaration(None=RayTracingQualitySwitch(Normal=EyeAdaptationInverse(LightValueInput=@n31, AlphaInput=@n32).EyeAdaptationInverse, RayTraced=Multiply(A=@n30, B=@n33))) {name=Emissive_1, nodeColor=<rgba 1 0.5859 0 1>}
  n34 = ScalarParameter() {group=SL_S30_Hexagon, parameterName=OutlineThreshhold}
  n35 = Power(Base=Fresnel(ExponentIn=Constant() {r=1}, Normal=VertexNormalWS()) {baseReflectFraction=0, exponent=0.5}, Exp=@n1)
  NamedRerouteDeclaration(None=Multiply(
    A=SmoothStep(Min=@n34, Value=@n35) {constMin=0.1}
    B=ScalarParameter() {defaultValue=2, group=SL_S30_Hexagon, parameterName=OutlineIntensity}
   )) {name=Fresnel_1, nodeColor=<rgba 0.9688 1 0 1>}
  n36 = TextureSample(UVs=NamedRerouteUsage()) {samplerType=SAMPLERTYPE_Normal, texture=/Game/Global/Materials/MaterialGym/Materials/HexDisplacement/Textures/T_Hexagon_Displacement_N}
  n37 = ScalarParameter() {defaultValue=1, group=STGridFX, parameterName=AdditionalGridStrength}
  NamedRerouteDeclaration(None=MaterialFunctionCall(Normal=@n36.RGB, Flatness=OneMinus(None=@n37)) {materialFunction=&FlattenNormal}.Result) {name=Normal_OUT, nodeColor=<rgba 1 0 0.4219 1>}
  n38 = Multiply(
   A=Multiply(A=TextureCoordinate(), B=ScalarParameter() {defaultValue=4, group=SL_S30_Hexagon, parameterName=Hex_Tiling}) {constB=5}
   B=ComponentMask(None=VectorParameter() {group=SL_S30_Hexagon, parameterName=Hex_TilingRatio, defaultValue=<rgba 0.85 1 0 1>}.RGB) {g=true, r=true}
  )
  n39 = TransformPosition(None=Constant3Vector()) {transformSourceType=TRANSFORMPOSSOURCE_Instance, transformType=TRANSFORMPOSSOURCE_World}
  n40 = Constant3Vector() {constant=<rgba 12.9898 78.841 37.718 0>}
  n41 = Multiply(A=Sine(None=DotProduct(A=@n39, B=@n40))) {constB=43758.5469}
  n42 = Multiply(B=NamedRerouteDeclaration(None=Frac(None=@n41)) {name=ObjectLocationBasedHash, nodeColor=<rgba 0 1 0.8828 1>}) {constA=2.2}
  n43 = TextureSample(UVs=NamedRerouteDeclaration(None=Add(A=@n38, B=@n42)) {name=Hex_UV, nodeColor=<rgba 0 1 0.4609 1>}) {
   samplerSource=SSM_Wrap_WorldGroupSettings
   texture=/Game/Global/Materials/MaterialGym/Materials/HexDisplacement/Textures/T_Hexagon_Displacement_Mask_V2
  }
  NamedRerouteDeclaration(None=@n43.R) {name=HexPattern_Greyscale, nodeColor=<rgba 0.125 1 0 1>}
  n44 = NamedRerouteUsage()
  n45 = Saturate(None=DotProduct(A=CameraVectorWS(), B=Transform(None=@n44) {transformSourceType=TRANSFORMSOURCE_Local}))
  n46 = ScalarParameter() {defaultValue=0.25, parameterName=Hash_ViewDependant_Power}
  NamedRerouteDeclaration(None=OneMinus(None=Power(Base=@n45, Exp=@n46))) {name=Hash_ViewDependant, nodeColor=<rgba 0.7578 0 1 1>}
  NamedRerouteDeclaration(None=Power(Base=OneMinus(None=@n43.G)) {constExponent=8}) {name=HexPattern_CenterGradient, nodeColor=<rgba 0.125 1 0 1>}
  n47 = LocalPosition()
  n48 = ComponentMask(None=@n47.XYZ) {r=true}
  n49 = Saturate(None=ScalarParameter() {bUseCustomPrimitiveData=true, parameterName=Use Vertical gradient?, primitiveDataIndex=11})
  n50 = Divide(
   A=Add(
    A=LinearInterpolate(A=@n47.Z, B=@n48, Alpha=@n49)
    B=ScalarParameter() {bUseCustomPrimitiveData=true, parameterName=Gradient Start Pos, primitiveDataIndex=1}
   )
   B=Add(A=ScalarParameter() {bUseCustomPrimitiveData=true, parameterName=Gradient Size, primitiveDataIndex=2})
  )
  n51 = LinearInterpolate(
   A=@n50
   B=Abs(None=@n50)
   Alpha=Saturate(None=ScalarParameter() {bUseCustomPrimitiveData=true, parameterName=Double gradient?, primitiveDataIndex=3})
  )
  n52 = ScalarParameter() {bUseCustomPrimitiveData=true, parameterName=Invert gradient?, primitiveDataIndex=12}
  n53 = Saturate(None=LinearInterpolate(A=@n51, B=OneMinus(None=@n51), Alpha=Saturate(None=@n52)))
  n54 = ScalarParameter() {bUseCustomPrimitiveData=true, parameterName=Use Gradient?}
  NamedRerouteDeclaration(None=LinearInterpolate(B=LinearInterpolate(Alpha=@n53) {constA=0.5, constB=1.2}, Alpha=Saturate(None=@n54)) {constA=1, constB=1.2}) {name=CustomPrimitiveDateMask_Out, nodeColor=<rgba 0 0.1797 1 1>}
  n55 = Multiply(A=Time(), B=ScalarParameter() {defaultValue=0.3, group=SL_S30_Hexagon, parameterName=HexIdleDisplacementSpeed})
  n56 = Sine(None=Subtract(A=NamedRerouteUsage(), B=NamedRerouteDeclaration(None=@n55) {name=Time_1, nodeColor=<rgba 0.9219 0 1 1>}))
  n57 = ScalarParameter() {defaultValue=0.1, group=SL_S30_Hexagon, parameterName=HexTileDepthColor}
  n58 = LinearInterpolate(A=Multiply(A=@n56, B=@n57) {constB=0.1}, B=Multiply(A=OneMinus(None=@n56), B=@n57) {constB=0.1}, Alpha=NamedRerouteUsage())
  NamedRerouteDeclaration(None=Multiply(A=Saturate(None=@n58), B=ScalarParameter() {defaultValue=1, group=MyGroup, parameterName=HexOverallIntensity})) {name=AnimatedHexHash, nodeColor=<rgba 0.5938 0 1 1>}
  NamedRerouteDeclaration(None=@n43.B) {name=ScratchMask, nodeColor=<rgba 0.3125 0 1 1>}
  NamedRerouteUsage()
  Transform(None=@n44) {transformSourceType=TRANSFORMSOURCE_Local, transformType=TRANSFORM_Tangent}
  Transform(None=@n44)
 ]}
