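# M_Hexagon_Darknexus is a Default Lit material for the Dark Nexus hexagon set art.
# It uses Material Attributes and holds no shading logic of its own, only a chain of three material functions.
# MF_Hexagon_Darknexus supplies the surface, which passes through MF_GlobalEmissiveMultiplier_MPC for global emissive scaling and then MF_FN_DestructibleDamage for destructible damage.
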Material /Hera/Sets/DarkNexus/TechArt/M_Hexagon_Darknexus
graph {
 properties {bUseMaterialAttributes=true, pixelDepthOffsetMode=PDOM_Legacy, usedShadingModels=Default Lit}
 nodes [MaterialFunctionCall(Material_In=MaterialFunctionCall!(In=MaterialFunctionCall!(In_1) {materialFunction=&/MF_Hexagon_Darknexus}.Result) {materialFunction=/Game/Global/MaterialFunctions/MF_GlobalEmissiveMultiplier_MPC}.Result) {materialFunction=&MF_FN_DestructibleDamage}]
}
